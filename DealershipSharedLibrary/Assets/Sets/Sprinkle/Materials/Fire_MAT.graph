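# Fire material for Niagara mesh particles: unlit, masked and two-sided.
# Two panning noise normal samples distort a vertical UV gradient, which is floored and combined with two scrolling Flame2 channels to build red-orange fire colours.
# A separate branch feeds vertex-coloured Flame3 into MF_FX_EmissiveAdjustments.
# Another builds a time-animated sine vector weighted by inverted V; the graph shows no connections to material outputs.
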Material /DealershipSharedLibrary/Assets/Sets/Sprinkle/Materials/Fire_MAT
graph {
 properties {
  bAutomaticallySetUsageInEditor=false
  bUsedWithNiagaraMeshParticles=true
  blendMode=BLEND_Masked
  shadingModel=MSM_Unlit
  twoSided=true
  usedShadingModels=Unlit
 }
 nodes [
  n1 = ComponentMask(None=TextureCoordinate()) {g=true}
  n2 = Time()
  n3 = TextureSample(UVs=Panner(Coordinate=TextureCoordinate() {uTiling=0.7, vTiling=0.1}, Time=@n2) {speedX=-0.3, speedY=0.1}) {samplerType=SAMPLERTYPE_Normal, texture=/Game/Global/Textures/Default/Noise/T_Noise_Normal_01}
  n4 = TextureSample(UVs=Panner(Coordinate=TextureCoordinate() {uTiling=0.4, vTiling=0.1}, Time=@n2) {speedX=-0.6, speedY=0.3}) {samplerType=SAMPLERTYPE_Normal, texture=&T_Noise_Normal_01}
  n5 = Add(A=ComponentMask(None=Add(A=@n3.RGB, B=@n4.RGB)) {g=true, r=true}, B=TextureCoordinate())
  n6 = Add(A=Power(Base=@n1) {constExponent=5}, B=Multiply(A=ComponentMask(None=@n5) {g=true}, B=Power(Base=@n1) {constExponent=0.1}))
  n7 = TextureCoordinate() {uTiling=1.5, vTiling=1.5}
  n8 = Add(
   A=TextureSample(UVs=Panner(Coordinate=@n7) {speedX=-0.3, speedY=2}) {texture=/DealershipSharedLibrary/Assets/Sets/Sprinkle/Textures/Flame2_TXT}.R
   B=TextureSample(UVs=Panner(Coordinate=@n7) {speedX=0.3, speedY=2}) {texture=&Flame2_TXT}.B
  )
  n9 = Clamp(None=Add(A=Floor(None=@n6), B=@n8))
  n10 = Constant3Vector() {constant=<rgba 1 0 0 0>}
  n11 = Constant3Vector() {constant=<rgba 1 0.4174 0 0>}
  Add(A=Multiply(A=Constant3Vector() {constant=<rgba 2 0.3 0 0>}, B=@n9), B=Add(A=Multiply(A=@n8, B=@n10) {constB=5}, B=Subtract(A=@n11, B=@n9)))
  MaterialFunctionCall(Emissive=Multiply(A=VertexColor(), B=TextureSample() {texture=/DealershipSharedLibrary/Assets/Sets/Sprinkle/Textures/Flame3_TXT}.RGB)) {materialFunction=&MF_FX_EmissiveAdjustments}
  n12 = OneMinus(None=@n1)
  n13 = Constant()
  n14 = AppendVector(A=@n13, B=@n13)
  n15 = Power(Base=@n12)
  n16 = Time()
  n17 = Sine(None=Add(A=Divide(A=@n15) {constB=0.8}, B=Multiply(A=@n16) {constB=-2}))
  Multiply(A=@n12, B=Multiply(A=AppendVector(A=@n14, B=@n17)) {constB=5}) {constB=5}
 ]
}
